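; Looping Niagara gameplay cue for the prefire build-up of the HighTower Date Ball Lightning ability.
; On loop start it caches the target pawn and reads the Ultra state from that pawn's Chain Lightning weapon.
; It then plays a first- or third-person build sound through the pawn's SoundLibrary.
; Each particle component gets Ultra and Summon flags, an Ultra-dependent scale and an index-based offset that mirrors the first two around the origin.
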(event EventOnLoopingStart(Niagara)
  (params (MyTarget :type Actor%20Object%20Reference) (Parameters :type Gameplay%20Cue%20Parameters%20Structure%20%28by%20ref%29) (ParticleComponents :type Array%20of%20Niagara%20Particle%20System%20Component%20Object%20References) (AudioComponents :type Array%20of%20Audio%20Component%20Object%20References))
  (bind _is_ultra (Variables|Default|GetIs%20Ultra))
  (|CallParentFunction MyTarget Parameters ParticleComponents AudioComponents)
  (Utilities|FlowControl|Sequence
    (:then_0
      (bind _asfort_player_pawn_athena (Utilities|Casting|CastToFortPlayerPawnAthena MyTarget))
      (bind _output_get (Variables|Default|SetPawn _asfort_player_pawn_athena))
      (bind _asb_high_tower_date_chain_lightning (Utilities|Casting|CastToB_HighTower_Date_ChainLightning (Variables|/Script/FortniteGame.FortPawn|GetCurrentWeapon _output_get)))
      (Variables|Default|SetIs%20Ultra (Variables|/HighTower/Items/Date/ChainLightning/B_HighTower_Date_ChainLightning.B_HighTower_Date_ChainLightning_C|GetIsUltra _asb_high_tower_date_chain_lightning)))
    (:then_1
      (Utilities|FlowControl|Sequence
        (:then_0
          (CallFunction|PlayBuildSound))
        (:then_1
          (for (_array_element _array_index) ParticleComponents
            (Effects|Components|ParticleSystem|SetBoolParameter _array_element "IsUltra" _is_ultra)
            (Transformation|SetWorldScale3D _array_element (Math|Conversions|ToVector(Float) (Math|Float|SelectFloat 0.16 0.095 _is_ultra)))
            (Transformation|SetRelativeLocation _array_element (Math|Vector|MakeVector (Utilities|Select :Index _array_index :Option%200 5.0 :Option%201 -5.0) (Utilities|Select :Index _array_index :Option%200 -5.0 :Option%201 5.0) (Utilities|Select :Index _array_index :Option%200 -5.0 :Option%201 5.0)) false false)
            (Effects|Components|ParticleSystem|SetBoolParameter _array_element "IsSummon" true)))))))

(fn PlayBuildSound
  (bind _pawn (Variables|Default|GetPawn)
    (:then
      (SoundLibrary|PlayEvent (SoundLibrary|GetSoundLibraryComponent _pawn "(TagName=\"\")" :!class /Script/FortniteGame.FortPawn) (select (Pawn|IsLocallyViewed _pawn) "(TagName=\"SoundLibrary.HighTower.Date.BallLightning.Build.1P\")" "(TagName=\"SoundLibrary.HighTower.Date.BallLightning.Build.3P\")")))
    (:else)))

(fn ConstructionScript)
