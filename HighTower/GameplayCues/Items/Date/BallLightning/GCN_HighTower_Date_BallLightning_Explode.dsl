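; Burst GameplayCue for the HighTower Date Ball Lightning explosion, adjusting the spawned particle effect and showing a gunshot-type sound indicator at the target.
; A vertical line trace around the cue location sets the effect's WaterFX flag when it hits a water body.
; The effect is rotated to the impact normal, falling back to the cue normal.
; If the physics field console variable is enabled, it also spawns a radial force field actor there.
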(fn OnBurst(EitherVFX-preferthisversion)
  (params (MyTarget :type Actor%20Object%20Reference) (Parameters :type Gameplay%20Cue%20Parameters%20Structure%20%28by%20ref%29) (ParticleComponents :type Array%20of%20FXSystem%20Component%20Object%20References) (AudioComponents :type Array%20of%20Audio%20Component%20Object%20References) (BurstCameraShakeInstance :type Camera%20Shake%20Base%20Object%20Reference) (BurstDecalInstance :type Decal%20Actor%20Object%20Reference))
  (bind (_normalizedmagnitude _rawmagnitude _effectcontext _matchedtagname _originaltag _aggregatedsourcetags _aggregatedtargettags _location _normal) (Ability|GameplayCue|BreakGameplayCueParameters Parameters))
  (bind _output (Utilities|Array|Get(acopy) ParticleComponents))
  (Fortnite|ShowSoundIndicatorLocal :WorldContextObject MyTarget :TrackedActor MyTarget :Location (Transformation|GetActorLocation MyTarget) :MaxAudibleDistance (ScalableFloat|GetValueatLevel (Variables|Default|GetRow_AVIndicatorDistance)) :IndicatorType "Gunshot" :Instigator MyTarget :Tint "(R=0.000000,G=0.000000,B=0.000000,A=0.000000)" :AttenuationMaxRangeMultiplier 1.0 :AttenuationIntensityMultiplier 1.0)
  (Effects|Components|ParticleSystem|SetBoolParameter _output "IsUltra" false)
  (bind (_outhit _returnvalue) (Collision|LineTraceForObjects 0 (+ _location "0.000000,0.000000,300.000000") (+ _location "0.000000,0.000000,-300.000000") (Variables|Default|GetObject%20Types)))
  (Effects|Components|ParticleSystem|SetBoolParameter _output "WaterFX" (.bSuccess (Utilities|Casting|CastToWaterBody (.HitActor (Collision|BreakHitResult _outhit)))))
  (if _returnvalue
    (Transformation|SetWorldRotation _output (Math|Rotator|MakeRotfromX (.ImpactNormal (Collision|BreakHitResult _outhit))) false false)
    (else
      (Transformation|SetWorldRotation _output (Math|Rotator|MakeRotfromX _normal) false false)))
  (bind _returnvalue_2 (Development|GetConsoleVariableIntValue "r.PhysicsField.EnableField"))
  (if (== _returnvalue_2 1)
    (Game|SpawnActorfromClass 0 "/Game/Effects/Niagara/Blueprints/PhsyicsFields/FN_RadialForce.FN_RadialForce_C" (Math|Conversions|ToTransform(Vector) _location) "Undefined" "MultiplyWithRoot" 0 1.0 0.75 1e+03 1.7 0.0 0)
    (return)))
