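; Burst gameplay cue for firing the HighTower Date Ball Lightning item, which adds an on-screen sound indicator on top of the parent cue's burst behavior.
; It calls the parent OnBurst, then casts the target to a Fort player pawn.
; For pawns not locally controlled, it shows a gunshot-type sound indicator at the pawn's location, with audible distance read from a scalable float.
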(fn OnBurst(EitherVFX-preferthisversion)
  (params (MyTarget :type Actor%20Object%20Reference) (Parameters :type Gameplay%20Cue%20Parameters%20Structure%20%28by%20ref%29) (ParticleComponents :type Array%20of%20FXSystem%20Component%20Object%20References) (AudioComponents :type Array%20of%20Audio%20Component%20Object%20References) (BurstCameraShakeInstance :type Camera%20Shake%20Base%20Object%20Reference) (BurstDecalInstance :type Decal%20Actor%20Object%20Reference))
  (|CallParentFunction MyTarget Parameters ParticleComponents AudioComponents BurstCameraShakeInstance BurstDecalInstance)
  (bind _asfort_player_pawn (Utilities|Casting|CastToFortPlayerPawn MyTarget)
    (:then
      (LocalVariables|SetPlayerPawn _asfort_player_pawn)
      (if (not (Pawn|IsLocallyControlled (LocalVariables|GetPlayerPawn)))
        (Fortnite|ShowSoundIndicatorLocal :WorldContextObject (LocalVariables|GetPlayerPawn) :TrackedActor (LocalVariables|GetPlayerPawn) :Location (Transformation|GetActorLocation (LocalVariables|GetPlayerPawn)) :MaxAudibleDistance (ScalableFloat|GetValueatLevel (Variables|Default|GetRow_AVIndicatorDistance)) :IndicatorType "Gunshot" :Instigator (LocalVariables|GetPlayerPawn) :Tint "(R=0.000000,G=0.000000,B=0.000000,A=0.000000)" :AttenuationMaxRangeMultiplier 1.0 :AttenuationIntensityMultiplier 1.0))
      (return))
    (:CastFailed
      (return))))
