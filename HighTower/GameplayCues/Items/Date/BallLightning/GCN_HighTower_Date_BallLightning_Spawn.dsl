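; Gameplay cue notify for the burst effect when the HighTower Date Ball Lightning item spawns its attack.
; On burst, it plays a 1P or 3P fire sound through the instigator pawn's SoundLibrary, based on whether that pawn is locally viewed.
; It attaches the first particle component to the pawn's mesh, rotates it to base aim, and sets its IsUltra parameter from the current Combo weapon.
; The PlayFireSound function is empty.
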(fn OnBurst(EitherVFX-preferthisversion)
  (params (MyTarget :type Actor%20Object%20Reference) (Parameters :type Gameplay%20Cue%20Parameters%20Structure%20%28by%20ref%29) (ParticleComponents :type Array%20of%20FXSystem%20Component%20Object%20References) (AudioComponents :type Array%20of%20Audio%20Component%20Object%20References) (BurstCameraShakeInstance :type Camera%20Shake%20Base%20Object%20Reference) (BurstDecalInstance :type Decal%20Actor%20Object%20Reference))
  (bind _output (Utilities|Array|Get(acopy) ParticleComponents))
  (bind _asfort_player_pawn (Utilities|Casting|CastToFortPlayerPawn (Game|GetInstigator MyTarget)))
  (Utilities|FlowControl|Sequence
    (:then_0
      (SoundLibrary|PlayEvent (SoundLibrary|GetSoundLibraryComponent _asfort_player_pawn "(TagName=\"\")" :!class /Script/FortniteGame.FortPawn) (select (Pawn|IsLocallyViewed _asfort_player_pawn) "(TagName=\"SoundLibrary.HighTower.Date.BallLightning.Fire.1P\")" "(TagName=\"SoundLibrary.HighTower.Date.BallLightning.Fire.3P\")"))
      (Transformation|AttachComponentToComponent _output (Variables|/Script/Engine.Character|GetMesh _asfort_player_pawn) "None" "KeepRelative" "KeepRelative" "KeepRelative" true)
      (Transformation|SetWorldRotation _output (Pawn|GetBaseAimRotation _asfort_player_pawn) false false))
    (:then_1
      (bind _asb_high_tower_date_combo (Utilities|Casting|CastToB_HighTower_Date_Combo (Variables|/Script/FortniteGame.FortPawn|GetCurrentWeapon _asfort_player_pawn)))
      (Effects|Components|ParticleSystem|SetBoolParameter _output "IsUltra" (Variables|/HighTower/Items/Date/ChainLightning/B_HighTower_Date_ChainLightning.B_HighTower_Date_ChainLightning_C|GetIsUltra _asb_high_tower_date_combo))
      (return))))

(fn PlayFireSound
  (params (Pawn :type Fort%20Player%20Pawn%20Object%20Reference)))
