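; Gameplay cue burst notify for the HighTower Date Ball Lightning item, named for its hit on world geometry.
; On burst, it reads the hit normal from the cue parameters and stores it locally.
; It then rotates the spawned burst decal's root component to a rotation built from that normal, adding a random full-circle roll so impact marks vary in orientation.
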(fn OnBurst(Cascade)
  (params (MyTarget :type Actor%20Object%20Reference) (Parameters :type Gameplay%20Cue%20Parameters%20Structure%20%28by%20ref%29) (ParticleComponents :type Array%20of%20Cascade%20Particle%20System%20Component%20Object%20References) (AudioComponents :type Array%20of%20Audio%20Component%20Object%20References) (BurstCameraShakeInstance :type Camera%20Shake%20Base%20Object%20Reference) (BurstDecalInstance :type Decal%20Actor%20Object%20Reference))
  (bind (_normalizedmagnitude _rawmagnitude _effectcontext _matchedtagname _originaltag _aggregatedsourcetags _aggregatedtargettags _location _normal) (Ability|GameplayCue|BreakGameplayCueParameters Parameters))
  (LocalVariables|SetHitNormal _normal)
  (Transformation|SetRelativeRotation (Variables|/Script/Engine.Actor|GetRootComponent BurstDecalInstance) (Math|Rotator|MakeRotator (+ (.roll (Math|Rotator|MakeRotfromX (LocalVariables|GetHitNormal))) (Math|Random|RandomFloatinRange 0.0 3.6e+02)) (.pitch (Math|Rotator|MakeRotfromX (LocalVariables|GetHitNormal))) (.yaw (Math|Rotator|MakeRotfromX (LocalVariables|GetHitNormal)))) false false)
  (return))
